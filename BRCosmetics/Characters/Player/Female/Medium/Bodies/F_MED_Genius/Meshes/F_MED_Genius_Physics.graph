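# Physics asset for the F_MED_Genius female medium player body, defining collision bodies and joints for the character and its cape.
# Capsules cover spine, pelvis, clavicles, arms, legs and feet under free-rotation joints, with a wide, flat box on root.
# Four cape chains of identical capsules hang from spine_05 with narrow twist and wider swing limits.
# Extra free-rotation constraints tie the four chain tips side by side, linking the cape's bottom edge.
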PhysicsAsset /BRCosmetics/Characters/Player/Female/Medium/Bodies/F_MED_Genius/Meshes/F_MED_Genius_Physics
bodies {
 calf_l [{shapeType=Capsule, length=31.8125, radius=10, center=<-30.1236 -1.4867 -0.0045>, rotation=<rot 89.8048 78.2753 83.2715>}]
 calf_r [{shapeType=Capsule, length=31.8125, radius=10, center=<30.1236 1.4867 0.0045>, rotation=<rot 89.8048 78.2754 83.2716>}]
 cape_02_l [{shapeType=Capsule, length=8, radius=14, center=<10 -12 0>, rotation=<rot 90 0 0>}]
 cape_02_r [{shapeType=Capsule, length=8, radius=14, center=<10 -12 0>, rotation=<rot 90 0 0>}]
 cape_03_l [{shapeType=Capsule, length=8, radius=14, center=<10 -12 0>, rotation=<rot 90 0 0>}]
 cape_03_r [{shapeType=Capsule, length=8, radius=14, center=<10 -12 0>, rotation=<rot 90 0 0>}]
 cape_04_l [{shapeType=Capsule, length=8, radius=14, center=<10 -12 0>, rotation=<rot 90 0 0>}]
 cape_04_r [{shapeType=Capsule, length=8, radius=14, center=<10 -12 0>, rotation=<rot 90 0 0>}]
 cape_05_l [{shapeType=Capsule, length=8, radius=14, center=<10 -12 0>, rotation=<rot 90 0 0>}]
 cape_05_r [{shapeType=Capsule, length=8, radius=14, center=<10 -12 0>, rotation=<rot 90 0 0>}]
 cape_06_l [{shapeType=Capsule, length=8, radius=14, center=<10 -12 0>, rotation=<rot 90 0 0>}]
 cape_06_r [{shapeType=Capsule, length=8, radius=14, center=<10 -12 0>, rotation=<rot 90 0 0>}]
 cape_07_l [{shapeType=Capsule, length=8, radius=14, center=<10 -12 0>, rotation=<rot 90 0 0>}]
 cape_07_r [{shapeType=Capsule, length=8, radius=14, center=<10 -12 0>, rotation=<rot 90 0 0>}]
 cape_mid_02_l [{shapeType=Capsule, length=8, radius=14, center=<10 -12 0>, rotation=<rot 90 0 0>}]
 cape_mid_02_r [{shapeType=Capsule, length=8, radius=14, center=<10 -12 0>, rotation=<rot 90 0 0>}]
 cape_mid_03_l [{shapeType=Capsule, length=8, radius=14, center=<10 -12 0>, rotation=<rot 90 0 0>}]
 cape_mid_03_r [{shapeType=Capsule, length=8, radius=14, center=<10 -12 0>, rotation=<rot 90 0 0>}]
 cape_mid_04_l [{shapeType=Capsule, length=8, radius=14, center=<10 -12 0>, rotation=<rot 90 0 0>}]
 cape_mid_04_r [{shapeType=Capsule, length=8, radius=14, center=<10 -12 0>, rotation=<rot 90 0 0>}]
 cape_mid_05_l [{shapeType=Capsule, length=8, radius=14, center=<10 -12 0>, rotation=<rot 90 0 0>}]
 cape_mid_05_r [{shapeType=Capsule, length=8, radius=14, center=<10 -12 0>, rotation=<rot 90 0 0>}]
 cape_mid_06_l [{shapeType=Capsule, length=8, radius=14, center=<10 -12 0>, rotation=<rot 90 0 0>}]
 cape_mid_06_r [{shapeType=Capsule, length=8, radius=14, center=<10 -12 0>, rotation=<rot 90 0 0>}]
 cape_mid_07_l [{shapeType=Capsule, length=8, radius=14, center=<10 -12 0>, rotation=<rot 90 0 0>}]
 cape_mid_07_r [{shapeType=Capsule, length=8, radius=14, center=<10 -12 0>, rotation=<rot 90 0 0>}]
 clavicle_l [{shapeType=Capsule, length=7.4279, radius=10.2908, center=<12 0 0>, rotation=<rot 90 0 0>}]
 clavicle_r [{shapeType=Capsule, length=7.4279, radius=10.2908, center=<-12 0 0>, rotation=<rot 90 0 0>}]
 foot_l [{shapeType=Capsule, length=7.2679, radius=10, shapeName=foot_l_capsule, center=<-2.6767 6.5019 0.5755>, rotation=<rot 0 0 90>}]
 foot_r [{shapeType=Capsule, length=7.2679, radius=10, shapeName=foot_r_capsule, center=<2.6767 -6.5019 -0.5755>, rotation=<rot 0 0 90>}]
 lowerarm_l [{shapeType=Capsule, length=34.1377, radius=6, center=<20 0 0>, rotation=<rot 90 0 0>}]
 lowerarm_r [{shapeType=Capsule, length=34.1377, radius=6, center=<-20 0 0>, rotation=<rot 90 0 0>}]
 pelvis [{
   shapeType=Capsule
   length=15.404
   radius=12.3772
   shapeName=pelvis_capsule
   center=<0.6105 -1.7395 0.0447>
   rotation=<rot 1.7884 -90 -0.1808>
  }]
 root [{shapeType=Box, extentX=231.0625, extentY=231.0625, extentZ=12.375, center=<0 0 -4.1865>, rotation=<rot 90 0 0>}]
 spine_01 [{shapeType=Capsule, length=14.2231, radius=9.6495, shapeName=spine_01_capsule, center=<0 -2.3725 0>, rotation=<rot -0.0001 -90 0>}]
 spine_02 [{shapeType=Capsule, length=10.3451, radius=10, shapeName=spine_02_capsule, center=<0 -1.0735 0>, rotation=<rot 0 -90 -0.0004>}]
 spine_04 [{shapeType=Capsule, length=17.7338, radius=10.4736, shapeName=spine_05_capsule, center=<0 1.2737 -0.1463>, rotation=<rot 6.5521 -90 0.8535>}]
 spine_05 [{
   shapeType=Capsule
   length=17.7338
   radius=10.4736
   shapeName=spine_05_capsule
   center=<4.5712 -1.0954 -1.1078>
   rotation=<rot 6.5521 -90 0.8535>
  }]
 thigh_l [{shapeType=Capsule, length=26.25, radius=14, center=<-20 -3.5 0>, rotation=<rot 90 0 0>}]
 thigh_r [{shapeType=Capsule, length=26.25, radius=14, center=<20 3.5 0>, rotation=<rot 90 0 0>}]
 upperarm_l [{shapeType=Capsule, length=34.1377, radius=6, center=<20 0 0>, rotation=<rot 90 0 0>}]
 upperarm_r [{shapeType=Capsule, length=34.1377, radius=6, center=<-20 0 0>, rotation=<rot 90 0 0>}]
}
constraints [
 {
  bone1Name=pelvis
  bone2Name=root
  swing1LimitDegrees=45
  swing1Motion=Free
  swing2LimitDegrees=45
  swing2Motion=Free
  twistLimitDegrees=45
  twistMotion=Free
 }
 {
  bone1Name=spine_01
  bone2Name=pelvis
  swing1LimitDegrees=45
  swing1Motion=Free
  swing2LimitDegrees=45
  swing2Motion=Free
  twistLimitDegrees=45
  twistMotion=Free
 }
 {
  bone1Name=spine_02
  bone2Name=spine_01
  swing1LimitDegrees=45
  swing1Motion=Free
  swing2LimitDegrees=45
  swing2Motion=Free
  twistLimitDegrees=45
  twistMotion=Free
 }
 {
  bone1Name=spine_04
  bone2Name=spine_02
  swing1LimitDegrees=45
  swing1Motion=Free
  swing2LimitDegrees=45
  swing2Motion=Free
  twistLimitDegrees=45
  twistMotion=Free
 }
 {
  bone1Name=spine_05
  bone2Name=spine_04
  swing1LimitDegrees=45
  swing1Motion=Free
  swing2LimitDegrees=45
  swing2Motion=Free
  twistLimitDegrees=45
  twistMotion=Free
 }
 {bone1Name=cape_02_l, bone2Name=spine_05, swing1LimitDegrees=45, swing2LimitDegrees=60, twistLimitDegrees=5}
 {bone1Name=cape_03_l, bone2Name=cape_02_l, swing1LimitDegrees=45, swing2LimitDegrees=60, twistLimitDegrees=5}
 {bone1Name=cape_04_l, bone2Name=cape_03_l, swing1LimitDegrees=45, swing2LimitDegrees=60, twistLimitDegrees=5}
 {bone1Name=cape_05_l, bone2Name=cape_04_l, swing1LimitDegrees=45, swing2LimitDegrees=60, twistLimitDegrees=5}
 {bone1Name=cape_06_l, bone2Name=cape_05_l, swing1LimitDegrees=45, swing2LimitDegrees=60, twistLimitDegrees=5}
 {bone1Name=cape_07_l, bone2Name=cape_06_l, swing1LimitDegrees=45, swing2LimitDegrees=60, twistLimitDegrees=5}
 {bone1Name=cape_02_r, bone2Name=spine_05, swing1LimitDegrees=45, swing2LimitDegrees=60, twistLimitDegrees=5}
 {bone1Name=cape_03_r, bone2Name=cape_02_r, swing1LimitDegrees=45, swing2LimitDegrees=60, twistLimitDegrees=5}
 {bone1Name=cape_04_r, bone2Name=cape_03_r, swing1LimitDegrees=45, swing2LimitDegrees=60, twistLimitDegrees=5}
 {bone1Name=cape_05_r, bone2Name=cape_04_r, swing1LimitDegrees=45, swing2LimitDegrees=60, twistLimitDegrees=5}
 {bone1Name=cape_06_r, bone2Name=cape_05_r, swing1LimitDegrees=45, swing2LimitDegrees=60, twistLimitDegrees=5}
 {bone1Name=cape_07_r, bone2Name=cape_06_r, swing1LimitDegrees=45, swing2LimitDegrees=60, twistLimitDegrees=5}
 {bone1Name=cape_mid_02_l, bone2Name=spine_05, swing1LimitDegrees=45, swing2LimitDegrees=60, twistLimitDegrees=5}
 {bone1Name=cape_mid_03_l, bone2Name=cape_mid_02_l, swing1LimitDegrees=45, swing2LimitDegrees=60, twistLimitDegrees=5}
 {bone1Name=cape_mid_04_l, bone2Name=cape_mid_03_l, swing1LimitDegrees=45, swing2LimitDegrees=60, twistLimitDegrees=5}
 {bone1Name=cape_mid_05_l, bone2Name=cape_mid_04_l, swing1LimitDegrees=45, swing2LimitDegrees=60, twistLimitDegrees=5}
 {bone1Name=cape_mid_06_l, bone2Name=cape_mid_05_l, swing1LimitDegrees=45, swing2LimitDegrees=60, twistLimitDegrees=5}
 {bone1Name=cape_mid_07_l, bone2Name=cape_mid_06_l, swing1LimitDegrees=45, swing2LimitDegrees=60, twistLimitDegrees=5}
 {bone1Name=cape_mid_02_r, bone2Name=spine_05, swing1LimitDegrees=45, swing2LimitDegrees=60, twistLimitDegrees=5}
 {bone1Name=cape_mid_03_r, bone2Name=cape_mid_02_r, swing1LimitDegrees=45, swing2LimitDegrees=60, twistLimitDegrees=5}
 {bone1Name=cape_mid_04_r, bone2Name=cape_mid_03_r, swing1LimitDegrees=45, swing2LimitDegrees=60, twistLimitDegrees=5}
 {bone1Name=cape_mid_05_r, bone2Name=cape_mid_04_r, swing1LimitDegrees=45, swing2LimitDegrees=60, twistLimitDegrees=5}
 {bone1Name=cape_mid_06_r, bone2Name=cape_mid_05_r, swing1LimitDegrees=45, swing2LimitDegrees=60, twistLimitDegrees=5}
 {bone1Name=cape_mid_07_r, bone2Name=cape_mid_06_r, swing1LimitDegrees=45, swing2LimitDegrees=60, twistLimitDegrees=5}
 {
  bone1Name=clavicle_l
  bone2Name=spine_05
  swing1LimitDegrees=45
  swing1Motion=Free
  swing2LimitDegrees=45
  swing2Motion=Free
  twistLimitDegrees=45
  twistMotion=Free
 }
 {
  bone1Name=upperarm_l
  bone2Name=clavicle_l
  swing1LimitDegrees=45
  swing1Motion=Free
  swing2LimitDegrees=45
  swing2Motion=Free
  twistLimitDegrees=45
  twistMotion=Free
 }
 {
  bone1Name=lowerarm_l
  bone2Name=upperarm_l
  swing1LimitDegrees=45
  swing1Motion=Free
  swing2LimitDegrees=45
  swing2Motion=Free
  twistLimitDegrees=45
  twistMotion=Free
 }
 {
  bone1Name=clavicle_r
  bone2Name=spine_05
  swing1LimitDegrees=45
  swing1Motion=Free
  swing2LimitDegrees=45
  swing2Motion=Free
  twistLimitDegrees=45
  twistMotion=Free
 }
 {
  bone1Name=upperarm_r
  bone2Name=clavicle_r
  swing1LimitDegrees=45
  swing1Motion=Free
  swing2LimitDegrees=45
  swing2Motion=Free
  twistLimitDegrees=45
  twistMotion=Free
 }
 {
  bone1Name=lowerarm_r
  bone2Name=upperarm_r
  swing1LimitDegrees=45
  swing1Motion=Free
  swing2LimitDegrees=45
  swing2Motion=Free
  twistLimitDegrees=45
  twistMotion=Free
 }
 {
  bone1Name=thigh_l
  bone2Name=pelvis
  swing1LimitDegrees=45
  swing1Motion=Free
  swing2LimitDegrees=45
  swing2Motion=Free
  twistLimitDegrees=45
  twistMotion=Free
 }
 {
  bone1Name=calf_l
  bone2Name=thigh_l
  swing1LimitDegrees=45
  swing1Motion=Free
  swing2LimitDegrees=45
  swing2Motion=Free
  twistLimitDegrees=45
  twistMotion=Free
 }
 {
  bone1Name=thigh_r
  bone2Name=pelvis
  swing1LimitDegrees=45
  swing1Motion=Free
  swing2LimitDegrees=45
  swing2Motion=Free
  twistLimitDegrees=45
  twistMotion=Free
 }
 {
  bone1Name=calf_r
  bone2Name=thigh_r
  swing1LimitDegrees=45
  swing1Motion=Free
  swing2LimitDegrees=45
  swing2Motion=Free
  twistLimitDegrees=45
  twistMotion=Free
 }
 {
  bone1Name=foot_l
  bone2Name=calf_l
  swing1LimitDegrees=45
  swing1Motion=Free
  swing2LimitDegrees=45
  swing2Motion=Free
  twistLimitDegrees=45
  twistMotion=Free
 }
 {
  bone1Name=foot_r
  bone2Name=calf_r
  swing1LimitDegrees=45
  swing1Motion=Free
  swing2LimitDegrees=45
  swing2Motion=Free
  twistLimitDegrees=45
  twistMotion=Free
 }
 {
  bone1Name=cape_07_l
  bone2Name=cape_mid_07_l
  swing1LimitDegrees=45
  swing1Motion=Free
  swing2LimitDegrees=45
  swing2Motion=Free
  twistLimitDegrees=45
  twistMotion=Free
 }
 {
  bone1Name=cape_mid_07_l
  bone2Name=cape_mid_07_r
  swing1LimitDegrees=45
  swing1Motion=Free
  swing2LimitDegrees=45
  swing2Motion=Free
  twistLimitDegrees=45
  twistMotion=Free
 }
 {
  bone1Name=cape_mid_07_r
  bone2Name=cape_07_r
  swing1LimitDegrees=45
  swing1Motion=Free
  swing2LimitDegrees=45
  swing2Motion=Free
  twistLimitDegrees=45
  twistMotion=Free
 }
]
