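# Post-process style AnimBlueprint for the F_MED_Genius body, adding secondary motion to its variant-toggled cape.
# The copied parent pose runs through a Control Rig, attach constraints, cape RigidBody sims chosen by backbling presence, and bone offsets into a cache.
# At low LOD, or without cloth and rigid bodies, movement-scaled trail controllers replace that result.
# The event graph derives movement states, drives cloth and gravity, and suspends cloth for backblings, emotes and certain vehicles.
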AnimBlueprint /BRCosmetics/Characters/Player/Female/Medium/Bodies/F_MED_Genius/Meshes/F_MED_Genius_AnimBP
target_skeleton=&/F_MED_Genius_Skeleton
graphs [
 AnimGraph {nodes [
   "Transform (Modify) Bone - Bone: dyn_calf_l_skirtAttach_front"!(ComponentPose, bAlphaBoolEnabled:bool=True, Translation="-0.631,5.264,0.342", Rotation="-44.178,-18.508,0.706", Scale="1.000,1.000,0.750") -> (Pose) {
    alphaInputType=Bool
    rotationMode=BMM_Replace
    rotationSpace=BCS_BoneSpace
    scaleMode=BMM_Replace
    scaleSpace=BCS_BoneSpace
    translationMode=BMM_Replace
    translationSpace=BCS_BoneSpace
    alphaBoolBlend {blendOption=Linear}
    boneToModify {boneName=dyn_calf_l_skirtAttach_front}
    rotation=<rot -44.178 -18.508 0.706>
    scale {z=0.75}
    translation=<-0.631 5.264 0.342>
   }
   n1 = Get BwdMovement() -> (BwdMovement:double)
   n2 = Trail controller - Bone: cape_mid_06_l(ComponentPose=Local To Component(LocalPose=Use cached pose 'RBAN Cache'()), Alpha=@n1) {
    bUsePlanarLimit=true
    chainLength=5
    debugLifeTime=0.1
    maxDeltaTime=0.333
    alphaScaleBiasClamp {bInterpResult=true}
    rotationLimits [{limitMax=<180 180 180>}, {limitMax=<180 180 180>}, {limitMax=<180 180 180>}, {limitMax=<180 180 180>}, {limitMax=<180 180 180>}]
    rotationOffsets [<0 0 0>, <0 0 0>, <0 0 0>, <0 0 0>, <0 0 0>]
    trailBone {boneName=cape_mid_06_l}
    trailRelaxationSpeed {editorCurveData {keys [<at -0.0134 19.3841 RCIM_Linear>,<at 0.5079 12.8986 RCIM_Linear>,<at 0.9955 9.6044 RCIM_Linear>]}}
   }
   n3 = Trail controller - Bone: cape_mid_06_r(ComponentPose=@n2, Alpha=@n1) {
    bUsePlanarLimit=true
    chainLength=5
    debugLifeTime=0.1
    maxDeltaTime=0.0333
    alphaScaleBiasClamp {bInterpResult=true}
    rotationLimits [
     {limitMax=<180 180 180>}
     {limitMax=<180 180 180>}
     {limitMax=<180 90 180>, limitMin=<-180 -90 -180>}
     {limitMax=<180 180 180>}
     {limitMax=<180 180 180>}
    ]
    rotationOffsets [<0 0 0>, <0 0 0>, <0 0 0>, <0 0 0>, <0 0 0>]
    trailBone {boneName=cape_mid_06_r}
    trailRelaxationSpeed {editorCurveData {keys [<at 0 15 RCIM_Linear>,<at 0.5289 13.1838 RCIM_Linear>,<at 1 6 RCIM_Linear>]}}
   }
   n4 = Trail controller - Bone: cape_07_l(ComponentPose=@n3, Alpha=@n1) {
    bUsePlanarLimit=true
    chainLength=5
    debugLifeTime=0.1
    maxDeltaTime=0.0333
    alphaScaleBiasClamp {bInterpResult=true}
    rotationLimits [
     {limitMax=<180 180 180>}
     {limitMax=<180 180 180>}
     {limitMax=<180 90 180>, limitMin=<-180 -90 -180>}
     {limitMax=<180 180 180>}
     {limitMax=<180 180 180>}
    ]
    rotationOffsets [<0 0 0>, <0 0 0>, <0 0 0>, <0 0 0>, <0 0 0>]
    trailBone {boneName=cape_07_l}
    trailRelaxationSpeed {editorCurveData {keys [<at 0 15 RCIM_Linear>,<at 0.5289 13.1838 RCIM_Linear>,<at 1 6 RCIM_Linear>]}}
   }
   n5 = Trail controller - Bone: cape_07_r(ComponentPose=@n4, Alpha=@n1) {
    bUsePlanarLimit=true
    chainLength=5
    debugLifeTime=0.1
    maxDeltaTime=0.0333
    alphaScaleBiasClamp {bInterpResult=true}
    rotationLimits [
     {limitMax=<180 180 180>}
     {limitMax=<180 180 180>}
     {limitMax=<180 90 180>, limitMin=<-180 -90 -180>}
     {limitMax=<180 180 180>}
     {limitMax=<180 180 180>}
    ]
    rotationOffsets [<0 0 0>, <0 0 0>, <0 0 0>, <0 0 0>, <0 0 0>]
    trailBone {boneName=cape_07_r}
    trailRelaxationSpeed {editorCurveData {keys [<at 0 15 RCIM_Linear>,<at 0.5289 13.1838 RCIM_Linear>,<at 1 6 RCIM_Linear>]}}
   }
   n6 = Get TrailControllerRot() -> (TrailControllerRot:Rotator)
   n7 = "Transform (Modify) Bone - Bone: cape_mid_02_l"(ComponentPose=@n5, Alpha=1, Translation="0,0,0", Rotation=@n6, Scale="1,1,1") {
    rotationMode=BMM_Additive
    rotationSpace=BCS_BoneSpace
    alphaBoolBlend {blendOption=Linear}
    boneToModify {boneName=cape_mid_02_l}
    rotation {yaw=-15}
   }
   n8 = "Transform (Modify) Bone - Bone: cape_mid_02_r"(ComponentPose=@n7, Alpha=1, Translation="0,0,0", Rotation=@n6, Scale="1,1,1") {
    rotationMode=BMM_Additive
    rotationSpace=BCS_BoneSpace
    alphaBoolBlend {blendOption=Linear}
    boneToModify {boneName=cape_mid_02_r}
    rotation {yaw=-15}
   }
   n9 = "Transform (Modify) Bone - Bone: cape_02_l"(ComponentPose=@n8, Alpha=1, Translation="0,0,0", Rotation=@n6, Scale="1,1,1") {
    rotationMode=BMM_Additive
    rotationSpace=BCS_BoneSpace
    alphaBoolBlend {blendOption=Linear}
    boneToModify {boneName=cape_02_l}
    rotation {yaw=-15}
   }
   n10 = "Transform (Modify) Bone - Bone: cape_02_r"(ComponentPose=@n9, Alpha=1, Translation="0,0,0", Rotation=@n6, Scale="1,1,1") {
    rotationMode=BMM_Additive
    rotationSpace=BCS_BoneSpace
    alphaBoolBlend {blendOption=Linear}
    boneToModify {boneName=cape_02_r}
    rotation {yaw=-15}
   }
   Output Pose(Result=Blend Poses by bool(
     BlendPose_0=Component To Local(ComponentPose=@n10)
     BlendPose_1=Use cached pose 'RBAN Cache'()
     bActiveValue=Get trailControllerConditions() -> (trailControllerConditions:bool)
    ) {blendType=Linear}) {layerGroup=None}
   n11 = Control Rig!(
    Source=Copy Pose From Mesh() {bUseAttachedParent=true}
    Alpha
    isCrouching:bool=Get bIsCrouching()
    R Side Movement:bool=Get R Side Movement() -> (R Side Movement:bool)
    L Side Movement:bool=Get L Side Movement() -> (L Side Movement:bool)
    isAccelerating2D:bool=Get bIsAccelerating2D()
    Backward movement:bool=Get Backward movement() -> (Backward movement:bool)
    Forward movement:bool=Get Forward movement() -> (Forward movement:bool)
    Is Targeting:bool=Get Is Targeting() -> (Is Targeting:bool)
    Is Crouch Sprinting:bool=Get Is Crouch Sprinting() -> (Is Crouch Sprinting:bool)
    isSkydiving:bool=Get bIsSkyDiving()
    Is Jumping:double=Get Is Jumping() -> (Is Jumping:double)
   ) -> (Pose) {
    alphaBoolBlend {blendOption=Linear}
    controlRigAssetReference {blueprintRigClass=&/F_MED_Genius_CtrlRig.F_MED_Genius_CtrlRig_C}
   }
   n12 = "Constraint - attach_cape (PPPPPPP)"!(
    ComponentPose=Local To Component(LocalPose=@n11)
    Alpha=1
    ConstraintWeights_0=0.073
    ConstraintWeights_1:float=0.0051
    ConstraintWeights_2:float=0.0051
    ConstraintWeights_3:float=0.4357
    ConstraintWeights_4:float=0.2241
    ConstraintWeights_5:float=0.0328
    ConstraintWeights_6:float=0.2241
   ) -> (Pose) {
    alphaBoolBlend {blendOption=Linear}
    boneToModify {boneName=attach_cape}
    constraintSetup [
     {targetBone {boneName=spine_04}}
     {targetBone {boneName=upperarm_l}}
     {targetBone {boneName=upperarm_r}}
     {targetBone {boneName=spine_05}}
     {targetBone {boneName=clavicle_r}}
     {targetBone {boneName=spine_03}}
     {targetBone {boneName=clavicle_l}}
    ]
    constraintWeights [1,1,1,1,1,1,1]
   }
   n13 = "Constraint - attach_backpack (PPPPPPPP)"!(
    ComponentPose=@n12
    Alpha=1
    ConstraintWeights_0=0.2039
    ConstraintWeights_1:float=0.0129
    ConstraintWeights_2:float=0.0129
    ConstraintWeights_3:float=0.1636
    ConstraintWeights_4:float=0.1942
    ConstraintWeights_5:float=0.0007
    ConstraintWeights_6:float=0.2176
    ConstraintWeights_7:float=0.1942
   ) -> (Pose) {
    alphaBoolBlend {blendOption=Linear}
    boneToModify {boneName=attach_backpack}
    constraintSetup [
     {targetBone {boneName=spine_04}}
     {targetBone {boneName=upperarm_l}}
     {targetBone {boneName=upperarm_r}}
     {targetBone {boneName=spine_05}}
     {targetBone {boneName=clavicle_r}}
     {targetBone {boneName=spine_02}}
     {targetBone {boneName=spine_03}}
     {targetBone {boneName=clavicle_l}}
    ]
    constraintWeights [1,1,1,1,1,1,1,1]
   }
   n14 = "Transform (Modify) Bone - Bone: attach_backpack"!(
    ComponentPose=@n13
    bAlphaBoolEnabled:bool=NOT Boolean(A=Get capeOn() -> (capeOn:bool))
    Translation="0,5.5,0"
    Rotation="0,0,-10"
    Scale="1,1,1"
   ) -> (Pose) {
    alphaInputType=Bool
    rotationMode=BMM_Additive
    rotationSpace=BCS_BoneSpace
    translationMode=BMM_Additive
    translationSpace=BCS_BoneSpace
    alphaBoolBlend {blendOption=Linear}
    boneToModify {boneName=attach_backpack}
    rotation {roll=-10}
    translation {y=5.5}
   }
   n15 = RigidBody!(
    ComponentPose=@n14
    bAlphaBoolEnabled:bool=AND Boolean(A=Get backbling_equipped() -> (backbling_equipped:bool), B=Get capeOn())
    OverrideWorldGravity:Vector=Get GravityOverride()
    ExternalForce
   ) -> (Pose) {
    alphaInputType=Bool
    bClampLinearTranslationLimitToRefPose=true
    bOverrideWorldGravity=true
    overridePhysicsAsset=&/F_MED_Genius_backbling_Physics
    simulationSpace=BaseBoneSpace
    alphaBoolBlend {blendOption=Linear}
    baseBoneRef {boneName=root}
    componentAppliedLinearAccClamp=<3000 3000 3000>
    componentLinearAccScale=<0.2 0.2 0.35>
    componentLinearVelScale=<0.2 0.2 0.35>
    overrideWorldGravity {z=-980}
    simSpaceSettings {
     maxAngularAcceleration=50
     maxAngularVelocity=1
     maxLinearAcceleration=300
     maxLinearVelocity=150
     velocityScaleZ=0
     worldAlpha=1
     externalLinearDragV=<0.2 0.2>
    }
   }
   n16 = Get backbling_equipped()
   n17 = RigidBody(ComponentPose=@n15, bAlphaBoolEnabled=AND Boolean(A=NOT Boolean(A=@n16), B=Get capeOn()), OverrideWorldGravity=Get GravityOverride()) {
    alphaInputType=Bool
    bClampLinearTranslationLimitToRefPose=true
    bOverrideWorldGravity=true
    overridePhysicsAsset=&/F_MED_Genius_Physics
    simulationSpace=BaseBoneSpace
    alphaBoolBlend {blendOption=Linear}
    baseBoneRef {boneName=root}
    componentAppliedLinearAccClamp=<3000 3000 3000>
    componentLinearAccScale=<0.2 0.2 0.35>
    componentLinearVelScale=<0.2 0.2 0.35>
    overrideWorldGravity {z=-980}
    simSpaceSettings {
     maxAngularAcceleration=50
     maxAngularVelocity=1
     maxLinearAcceleration=300
     maxLinearVelocity=150
     velocityScaleZ=0
     worldAlpha=1
     externalLinearDragV=<0.2 0.2>
    }
   }
   n18 = "Transform (Modify) Bone - Bone: cape_02_r"!(
    ComponentPose=@n17
    bAlphaBoolEnabled:bool=AND Boolean(A=Get bIsInFrontEnd(), B=Get bClothEnabled())
    Translation="0,0,0"
    Rotation="-0.000,0.000,20"
    Scale="1,1,1"
   ) -> (Pose) {alphaInputType=Bool, rotationMode=BMM_Additive, alphaBoolBlend {blendOption=Linear}, boneToModify {boneName=cape_02_r}, rotation {roll=20}}
   n19 = Get bIsSkyDiving()
   n20 = "Transform (Modify) Bone - Bone: cape_03_r"!(ComponentPose=@n18, bAlphaBoolEnabled:bool=@n19, Translation="0,0,0", Rotation="-15,-25.000,-0.000", Scale="1,1,1") -> (Pose) {
    alphaInputType=Bool
    rotationMode=BMM_Additive
    rotationSpace=BCS_BoneSpace
    translationMode=BMM_Additive
    translationSpace=BCS_BoneSpace
    alphaBoolBlend {blendOption=Linear}
    boneToModify {boneName=cape_03_r}
    rotation {pitch=-15, yaw=-25}
   }
   n21 = "Transform (Modify) Bone - Bone: cape_mid_02_l"!(ComponentPose=@n20, bAlphaBoolEnabled:bool=@n19, Translation="0,0,0", Rotation="15,-25,0", Scale="1,1,1") -> (Pose) {
    alphaInputType=Bool
    rotationMode=BMM_Additive
    rotationSpace=BCS_BoneSpace
    translationMode=BMM_Additive
    translationSpace=BCS_BoneSpace
    alphaBoolBlend {blendOption=Linear}
    boneToModify {boneName=cape_mid_02_l}
    rotation {pitch=15, yaw=-25}
   }
   n22 = "Transform (Modify) Bone - Bone: cape_mid_02_r"!(ComponentPose=@n21, bAlphaBoolEnabled:bool=@n19, Translation="0,0,0", Rotation="-15,-25.000,-0.000", Scale="1,1,1") -> (Pose) {
    alphaInputType=Bool
    rotationMode=BMM_Additive
    rotationSpace=BCS_BoneSpace
    translationMode=BMM_Additive
    translationSpace=BCS_BoneSpace
    alphaBoolBlend {blendOption=Linear}
    boneToModify {boneName=cape_mid_02_r}
    rotation {pitch=-15, yaw=-25}
   }
   n23 = "Transform (Modify) Bone - Bone: cape_03_l"!(ComponentPose=@n22, bAlphaBoolEnabled:bool=@n19, Translation="0,0,0", Rotation="15,-25,0", Scale="1,1,1") -> (Pose) {
    alphaInputType=Bool
    rotationMode=BMM_Additive
    rotationSpace=BCS_BoneSpace
    translationMode=BMM_Additive
    translationSpace=BCS_BoneSpace
    alphaBoolBlend {blendOption=Linear}
    boneToModify {boneName=cape_03_l}
    rotation {pitch=15, yaw=-25}
   }
   n24 = "Transform (Modify) Bone - Bone: cape_02_l"!(ComponentPose=@n23, bAlphaBoolEnabled:bool=True, Translation="0,-3,-0.000", Rotation="-0.000,-2,-0.000") -> (Pose) {
    alphaInputType=Bool
    rotationMode=BMM_Replace
    rotationSpace=BCS_BoneSpace
    scaleMode=BMM_Replace
    scaleSpace=BCS_BoneSpace
    translationMode=BMM_Replace
    translationSpace=BCS_BoneSpace
    alphaBoolBlend {blendOption=Linear}
    boneToModify {boneName=cape_02_l}
    rotation {yaw=-2}
    translation {y=-3}
   }
   n25 = "Transform (Modify) Bone - Bone: cape_02_r"!(ComponentPose=@n24, bAlphaBoolEnabled=True, Translation="0,-3,-0.000", Rotation="0.000,-2,0") -> (Pose) {
    alphaInputType=Bool
    rotationMode=BMM_Replace
    rotationSpace=BCS_BoneSpace
    scaleMode=BMM_Replace
    scaleSpace=BCS_BoneSpace
    translationMode=BMM_Replace
    translationSpace=BCS_BoneSpace
    alphaBoolBlend {blendOption=Linear}
    boneToModify {boneName=cape_02_r}
    rotation {yaw=-2}
    translation {y=-3}
   }
   n26 = "Transform (Modify) Bone - Bone: dyn_L_shoulderpad"!(ComponentPose=@n25, bAlphaBoolEnabled:bool=True, Translation="0,0,1", Rotation="0,0,0") -> (Pose) {
    alphaInputType=Bool
    rotationMode=BMM_Additive
    rotationSpace=BCS_BoneSpace
    translationMode=BMM_Additive
    translationSpace=BCS_BoneSpace
    alphaBoolBlend {blendOption=Linear}
    boneToModify {boneName=dyn_L_shoulderpad}
    translation {z=1}
   }
   n27 = "Transform (Modify) Bone - Bone: upperarm_l_pauldron"!(ComponentPose=@n26, bAlphaBoolEnabled:bool=True, Translation="-0.000,2,0", Rotation="0,0,0") -> (Pose) {
    alphaInputType=Bool
    rotationMode=BMM_Additive
    rotationSpace=BCS_BoneSpace
    translationMode=BMM_Additive
    translationSpace=BCS_BoneSpace
    alphaBoolBlend {blendOption=Linear}
    boneToModify {boneName=upperarm_l_pauldron}
    translation {y=2}
   }
   Save cached pose 'RBAN Cache'(Pose=Component To Local(ComponentPose=@n27)) {cacheName=RBAN Cache}
   Get Left_Back() -> (Left_Back:bool)
  ]}
 EventGraph {nodes [
   n1 = Get Dynamic Material Instance(SkeletalMeshComponent=GetPartSkeletalMeshComponent())
   > n2 = Branch(Condition=Get capeOn())
   > SetScalarParameterValue(self=@n1, ParameterName=Hide Element 01, Value=0)
   SetScalarParameterValue(execute=@n2.else, self=@n1, ParameterName=Hide Element 01, Value=1)
   Event OnBeginSkydiving()
   > Set TrailControllerRot(TrailControllerRot:Rotator="0,-65,0") -> (Output_Get:Rotator)
   > Delay(Duration=0.3)
   > ResetDynamics(InTeleportType=TeleportPhysics)
   > ResetDynamics(InTeleportType=ResetPhysics)
   Event OnCharacterCustomizationLoaded()
   > n3 = Branch(Condition=CharacterItemDefinitionHasAnyMetaTag(Tags=Make Array("[0]"=(TagName="Cosmetics.Variant.Property.Stage1"))))
   > Set capeOn(capeOn:bool=true) -> (Output_Get:bool)
   Set capeOn(execute=@n3.else, capeOn=false)
   Event OnBeginJumping()
   > Set Is Jumping(Is Jumping:double=1) -> (Output_Get:double)
   Event BlueprintInitializeAnimation()
   > Set SkelMesh(SkelMesh:Skeletal Mesh Object Reference=&/F_MED_Genius) -> (Output_Get:Skeletal Mesh Object Reference)
   > n4 = Cast To FortPlayerPawn(Object=TryGetFortPawn())
   > Set Pawn(Pawn:Fort Player Pawn Object Reference=@n4) -> (Output_Get:Fort Player Pawn Object Reference)
   Event OnEndJumping()
   > Delay(Duration=0.2)
   > Set Is Jumping(Is Jumping=0)
   Event OnEndSkydiving()
   > Set TrailControllerRot(TrailControllerRot="0,-10,0")
   Event BlueprintUpdateAnimation()
   > n5 = Sequence!(execute) -> (then_0, then_1, then_2:Exec, then_3:Exec, then_4:Exec, then_5:Exec, then_6:Exec, then_7:Exec, then_8:Exec)
   > Set cloth_disabled(cloth_disabled:bool=NOT Boolean(A=Get bClothEnabled())) -> (Output_Get:bool)
   > Set lowLODEnabled(lowLODEnabled:bool="integer >= integer"(A=Get CurrentLOD(), B=2)) -> (Output_Get:bool)
   n6 = NOT Boolean(A=Get bRigidBodyEnabled())
   n7 = NOT Boolean(A=Get bClothEnabled())
   > Set trailControllerConditions(trailControllerConditions:bool=OR Boolean(A=Get lowLODEnabled() -> (lowLODEnabled:bool), B=AND Boolean(A=@n7, B=@n6))) -> (Output_Get:bool)
   > Set Is Targeting(Is Targeting:bool=Get bIsTargeting()) -> (Output_Get:bool)
   > Set Is Crouch Sprinting(Is Crouch Sprinting:bool=Get bIsCrouchSprinting()) -> (Output_Get:bool)
   n8 = Gravity Override State Function DEPRECATED(
    execute=@n5.then_1
    CharacterPartAnimInstance=Self-Reference()
    GameTimeInSeconds=GetGameTimeInSeconds()
    GravityOverrideParams=Get Gravity Override Params()
   )
   > Set GravityOverride(GravityOverride=@n8)
   Branch(execute=@n5.then_3, Condition=Get bClothEnabled())
   > Cloth Controller Function DEPRECATED!(
    execute
    CharacterPartAnimInstance:Custom Character Part Anim Instance Object Reference=Self-Reference()
    ClothingAssetName:String=F_MED_Genius_Clothing
    Cloth Paramaters:Cloth Params Struct Structure=Get Cloth Parameters() -> (Cloth Parameters:Cloth Params Struct Structure)
    Cloth Core Settings:Cloth Core Setting Struct Structure=Get Cloth Core Settings() -> (Cloth Core Settings:Cloth Core Setting Struct Structure)
    ClothUpAxis:Vector="0,0,1"
    OnGround_Moving_SpeedLimit:double=4
    OnGround_Moving_AngularVel_Max:double=0.05
    OnGround_Moving_FictitiousVel_Max:double=0.6
    LinVelScale_FwdBwd_Clamp:double=0
    LinVelScale_UpDwn_Clamp:double=0.7
    DeltaTime:double=0
    GameTimeInSeconds:double=GetGameTimeInSeconds()
    WindGustsParams:Wind Gust Params Struct Structure
    WindGustLocalSettings:Wind Gust Local Setting Struct Structure
   ) -> (
    then
    CurrentStateOut:String
    Cloth Settings Out:Cloth Setting Struct Structure
    Cloth Core Settings Out:Cloth Core Setting Struct Structure
    WindGustSettings Out:Wind Gust Setting Struct Structure
    WindGustLocalSettings Out:Wind Gust Local Setting Struct Structure
    GustStrength - WindVelocityOut:Vector
    LinearVelVector-DELETE LATER:Vector
   )
   Branch(execute=@n5.then_4, Condition=OR Boolean(A=Get bIsInFrontEnd(), B=Get bIsPlayingEmote()))
   > Set BwdMovement(BwdMovement:double=1) -> (Output_Get:double)
   n9 = CheckForBackbling()
   n10 = Set backbling_equipped(execute=@n5.then_5, backbling_equipped:bool=@n9) -> (Output_Get:bool)
   > n11 = Branch(Condition=OR Boolean!(
     A=Get bVehicleSeatForcesCrouch() -> (bVehicleSeatForcesCrouch:bool)
     B=Get bIsGolfCartVehicleBackPassenger() -> (bIsGolfCartVehicleBackPassenger:bool)
     C:bool=Get bIsJackalVehicleDriver() -> (bIsJackalVehicleDriver:bool)
     D:bool=Get bIsInShoppingCart() -> (bIsInShoppingCart:bool)
     E:bool=Get bIsInCannon() -> (bIsInCannon:bool)
     F:bool=Get bIsMeatballPassenger() -> (bIsMeatballPassenger:bool)
     G:bool=@n9
     H:bool=@n10
     I:bool=Get bIsPlayingEmote()
    ) -> (ReturnValue))
   n12 = GetPartSkeletalMeshComponent()
   > Set ClothBlendWeight!(execute, ClothBlendWeight:float=0, self:Skeletal Mesh Component Object Reference=@n12) -> (then, Output_Get:float)
   > Do Once(Reset=@n14.then)
   > n13 = SuspendClothingSimulation(self:Skeletal Mesh Component Object Reference=@n12)
   Do Once(execute=@n15.then, Reset=@n13.then)
   > n14 = ResumeClothingSimulation(self:Skeletal Mesh Component Object Reference=@n12)
   n15 = Set ClothBlendWeight(execute=@n11.else, ClothBlendWeight=1, self=@n12)
   n16 = PlayerMovement(execute=@n5.then_6, AnimInstance=Self-Reference(), XDotProductCoefficient=1, YDotProductCoefficient=1, ZDotProductCoefficient=1)
   n17 = MapRangeClamped(Value=@n16.YMovement, InRangeA=0, InRangeB=1)
   n18 = "Absolute (Float)"(A=@n16)
   > Set BwdMovement(BwdMovement=MapRangeClamped(Value=float + float(A=@n18, B=@n17) {is=title K2Node_CommutativeAssociativeBinaryOperator}, InRangeA=0, InRangeB=1.4))
   > n19 = Branch(Condition=Get bIsSwimSprinting() -> (bIsSwimSprinting:bool))
   > Set TrailControllerRot(TrailControllerRot="0,-15, 0")
   n20 = Branch(execute=@n19.else, Condition=Get bIsSkyDiving())
   > n21 = Branch(Condition=Get bIsParachuteOpened())
   > Set TrailControllerRot(TrailControllerRot="0,-15, 0")
   > Set BwdMovement(BwdMovement=0.25)
   n22 = Branch(execute=@n21.else, Condition="float > float"(A=Get PawnVelocityZ(), B=50) {is=title K2Node_CallFunction})
   > Set BwdMovement(BwdMovement=float * float(A=Get BwdMovement(), B=0))
   > Set TrailControllerRot(TrailControllerRot="0,-25, 0")
   n23 = Branch(execute=@n22.else, Condition=Get bIsSkydiveDiveMode())
   > Set TrailControllerRot(TrailControllerRot="0,-15, 0")
   Set TrailControllerRot(execute=@n23.else, TrailControllerRot="0,-25, 0")
   Set TrailControllerRot(execute=@n20.else, TrailControllerRot="0,-12.5, 0")
   > Set BwdMovement(BwdMovement=float * float(A=Get BwdMovement(), B=1))
   Set Left_Back(
    execute=@n5.then_7
    Left_Back:bool=AND Boolean!(A=Get Backward movement(), B=Get L Side Movement(), C:bool=Get bIsCrouching()) -> (ReturnValue)
   ) -> (Output_Get:bool)
   Get!(execute:Exec=@n5.then_2) -> (then:Exec, else:Exec, SkelMesh:Skeletal Mesh Object Reference)
   > Get!(execute) -> (then, else, Pawn:Fort Player Pawn Object Reference)
   > n24 = PlayerMovement(AnimInstance=Self-Reference(), XDotProductCoefficient=100, YDotProductCoefficient=100, ZDotProductCoefficient=100)
   n25 = BreakVector(InVec=MakeVector(X=@n24, Y=@n24.YMovement))
   > Set L Side Movement(L Side Movement:bool="float < float"(A=@n25, B=-50)) -> (Output_Get:bool)
   > Set R Side Movement(R Side Movement:bool="float > float"(A=@n25, B=50) {is=title K2Node_PromotableOperator}) -> (Output_Get:bool)
   > Set Forward movement(Forward movement:bool="float > float"(A=@n25.Y, B=40)) -> (Output_Get:bool)
   > Set Backward movement(Backward movement:bool="float < float"(A=@n25.Y, B=-40)) -> (Output_Get:bool)
   SuspendClothingSimulation()
  ]}
]
variables [
 backbling_equipped {type=boolean}
 Backward movement {type=boolean}
 BwdMovement {type=number}
 capeOn {type=boolean}
 Cloth Core Settings {
  type=ClothCoreSettingStruct
  default {
   collisionThickness=1.5
   dampingCoefficient=0.1
   frictionCoefficient=0.5
   localDampingCoefficient=0.1
   numberIterations=2
   numberSubsteps=1
   selfCollisionThickness=0
   animDriveDamping=<0 1>
   animDriveStiffness=<0 1>
   areaStiffness=<1 1>
   bendingStiffness=<1 1>
   edgeStiffness=<1 1>
   tetherStiffness=<1 1>
  }
 }
 Cloth Parameters {
  type=ClothParamsStruct
  default {
   dBNO {
    angularVelocityScale=0.6
    fictitiousAngularScale=0.1
    jointName=pelvis
    dragCoefficient=<0.2 1>
    gravityOverride=<-980 400 0>
    liftCoefficient=<0 0.5>
    linearVelocityScale=<0.3 0.3 0.1>
   }
   dBNO_Carried {
    angularVelocityScale=0.6
    fictitiousAngularScale=0.1
    jointName=pelvis
    dragCoefficient=<0.2 1>
    gravityOverride=<-980 -200 0>
    liftCoefficient=<0 0.5>
    linearVelocityScale=<0.3 0.3 0.1>
   }
   dBNO_Impulse {
    angularVelocityScale=0
    fictitiousAngularScale=0
    jointName=pelvis
    dragCoefficient=<0 0.2>
    gravityOverride=<-3000 -2000 0>
    liftCoefficient=<0 0.1>
    linearVelocityScale=<0.3 0.1 0.3>
   }
   dBNO_RollBwd {
    angularVelocityScale=0
    fictitiousAngularScale=0
    jointName=pelvis
    dragCoefficient=<0 0.2>
    gravityOverride=<-2000 0 0>
    liftCoefficient=<0 0.1>
    linearVelocityScale=<0.3 0.1 0.3>
   }
   dBNO_RollFwd {
    angularVelocityScale=0
    fictitiousAngularScale=0
    jointName=pelvis
    dragCoefficient=<0 0.2>
    gravityOverride=<-980 0 0>
    liftCoefficient=<0 0.1>
    linearVelocityScale=<0.3 0.1 0.3>
   }
   dBNO_Tumbling {
    angularVelocityScale=0
    fictitiousAngularScale=0
    jointName=pelvis
    dragCoefficient=<0 0.2>
    gravityOverride=<-2500 -300 0>
    liftCoefficient=<0 0.1>
    linearVelocityScale=<0.3 0.1 0.3>
   }
   emoteOrMelee {
    angularVelocityScale=0.6
    fictitiousAngularScale=0.1
    jointName=root
    dragCoefficient=<2.5 2.5>
    gravityOverride=<0 0 -980>
    liftCoefficient=<0.5 0.5>
    linearVelocityScale=<0.6 0.6 0.5>
   }
   front_End {
    angularVelocityScale=0.05
    fictitiousAngularScale=0.6
    jointName=root
    dragCoefficient=<0.2 0.2>
    gravityOverride=<0 0 -980>
    liftCoefficient=<0.1 0.1>
    linearVelocityScale=<0.1 0.1 0.1>
   }
   onGround_Crouch_Moving {
    angularVelocityScale=0.1
    fictitiousAngularScale=0.3
    jointName=root
    dragCoefficient=<2 2>
    gravityOverride=<0 0 -980>
    liftCoefficient=<0.5 0.5>
    linearVelocityScale=<0.6 0.6 0.5>
   }
   onGround_Falling {
    angularVelocityScale=0.6
    fictitiousAngularScale=0
    jointName=root
    dragCoefficient=<0 0.2>
    gravityOverride=<0 -500 -980>
    liftCoefficient=<0 0.1>
    linearVelocityScale=<2 1 0.3>
   }
   onGround_Moving {
    angularVelocityScale=0.6
    fictitiousAngularScale=0.1
    jointName=root
    dragCoefficient=<2 2>
    gravityOverride=<0 0 -980>
    liftCoefficient=<0.5 0.5>
    linearVelocityScale=<0.6 0.6 0.5>
   }
   onGround_Standing {
    angularVelocityScale=0.05
    fictitiousAngularScale=1.2
    jointName=root
    dragCoefficient=<2.5 2.5>
    gravityOverride=<0 0 -980>
    liftCoefficient=<0.5 0.5>
    linearVelocityScale=<0.6 0.6 0.5>
   }
   rail_Grinding {
    angularVelocityScale=0.2
    fictitiousAngularScale=0
    jointName=root
    dragCoefficient=<0.1 0.2>
    gravityOverride=<50 -500 -400>
    liftCoefficient=<0 0.5>
    linearVelocityScale=<0.1 0.1 0.1>
   }
   riding_Moving {
    angularVelocityScale=0.05
    fictitiousAngularScale=1
    jointName=root
    dragCoefficient=<0.5 0.5>
    gravityOverride=<0 0 -980>
    liftCoefficient=<0.5 0.5>
    linearVelocityScale=<0.4 0.6 0.3>
   }
   riding_Moving_Targeting {
    angularVelocityScale=0.05
    fictitiousAngularScale=0.7
    jointName=root
    dragCoefficient=<0.5 0.5>
    gravityOverride=<0 0 -980>
    liftCoefficient=<0.5 0.5>
    linearVelocityScale=<0.2 0.6 0.2>
   }
   riding_Standing {
    angularVelocityScale=0.05
    fictitiousAngularScale=1
    jointName=root
    dragCoefficient=<0.5 0.5>
    gravityOverride=<0 -450 -980>
    liftCoefficient=<0.5 0.5>
    linearVelocityScale=<0.6 0.6 0.3>
   }
   skydive_DivingDown {
    angularVelocityScale=0.1
    fictitiousAngularScale=0
    jointName=pelvis
    dragCoefficient=<0.3 0.3>
    gravityOverride=<-450 -100 0>
    liftCoefficient=<0.9 0.9>
    linearVelocityScale=<0.1 0.1 0.1>
   }
   skydive_FloatingDown {
    angularVelocityScale=0.1
    fictitiousAngularScale=0
    jointName=root
    dragCoefficient=<0.3 0.3>
    gravityOverride=<0 -800 200>
    liftCoefficient=<1.3 1.3>
    linearVelocityScale=<0.2 0.2 0.1>
   }
   skydive_GoingUp {
    angularVelocityScale=0.1
    fictitiousAngularScale=0
    jointName=root
    dragCoefficient=<0.01 0.01>
    gravityOverride=<0 -2000 400>
    liftCoefficient=<0.05 0.05>
    linearVelocityScale=<0.1 0.1 0.01>
   }
   skydive_Parachuting {
    angularVelocityScale=0.6
    fictitiousAngularScale=0.1
    jointName=root
    dragCoefficient=<0.8 0.8>
    gravityOverride=<0 -600 -500>
    liftCoefficient=<0.8 0.8>
    linearVelocityScale=<0.2 0.2 0.3>
   }
   skydive_Surfing {
    angularVelocityScale=0.6
    fictitiousAngularScale=0.1
    jointName=root
    dragCoefficient=<1.5 1.5>
    gravityOverride=<0 -600 -500>
    liftCoefficient=<1 1>
    linearVelocityScale=<0.1 0.1 0.3>
   }
   swimming {
    angularVelocityScale=0.6
    fictitiousAngularScale=0
    jointName=root
    dragCoefficient=<2.5 2.5>
    gravityOverride=<0 0 0>
    liftCoefficient=<0.1 0.1>
    linearVelocityScale=<0.6 0.6 0.5>
   }
   tactical_Sprint {
    angularVelocityScale=0.6
    fictitiousAngularScale=0.1
    jointName=root
    dragCoefficient=<2 2>
    gravityOverride=<0 0 -980>
    liftCoefficient=<0.5 0.5>
    linearVelocityScale=<0.6 0.6 0.5>
   }
   ziplining {
    angularVelocityScale=0.6
    fictitiousAngularScale=0
    jointName=root
    dragCoefficient=<0 1.5>
    gravityOverride=<0 0 -980>
    liftCoefficient=<0 0.3>
    linearVelocityScale=<0.4 0.4 0.5>
   }
  }
 }
 cloth_disabled {type=boolean}
 DeltaTimeLocal {type=number}
 Forward movement {type=boolean}
 Gravity Override Params {type=GravityOverrideParamsStruct, default {onGround_Moving {gravityOverride=<0 0 -500>}}}
 GravityOverride {type=Vector}
 Is Crouch Sprinting {type=boolean}
 Is Jumping {type=number}
 Is Targeting {type=boolean}
 L Side Movement {type=boolean}
 Left_Back {type=boolean}
 lowLODEnabled {type=boolean}
 Pawn {type=FortPlayerPawn}
 R Side Movement {type=boolean}
 SkelMesh {type=SkeletalMesh, default=null}
 trailControllerConditions {type=boolean}
 TrailControllerRot {type=Rotator}
]
